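; Base parent Blueprint for CherrySmoke lootable chests, with hero-rank lock visuals, ambient audio and falling behavior.
; Playlist-ready setup re-applies loot tier as a temp fix, resets searched state and sets custom ping lifetime.
; UpdateLockStatusVFX drives Niagara, audio and sound indicator from rank; looting plays a burst and disables falling.
; When its linked actor is destroyed, a scalable-float-gated handler destroys the chest, first spawning a class-carrying projectile if falling is allowed.
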(event EventBeginPlay
  ; Begin Play
  (Class|FortAsyncActionCurrentPlaylistReady|PlaylistDataReadyAsync
    (:then
      (|CallParentFunction))
    (:OnReady
      ;;[0.5208 0.5208 0.5208 0.5] Temp fix for FORT-887358 revert once fixed
      (Container|SetSearchLootTierGroup (Variables|Container|GetSearchLootTierGroup))
      (Variables|Container|SetbAlreadySearched false)
      (Variables|Building|SetbDestroyOnPlayerBuildingPlacement false)
      (CallFunction|SetCustomPingTimerforMarkerDisplayInfo))))

(event Custom|RemoveVFXAndUnbindEvents
  ; Remove VFX and Unbind Events - Call or override
  (Components|DestroyComponent (Variables|Components|GetHeroRankUnlockedVFX))
  (bind _chest_ambient_sound (Variables|Default|GetChest_Ambient_Sound)
    (:then
      (Audio|Components|Audio|FadeOut _chest_ambient_sound 0.5 0.0 "Linear" :!class /Script/Engine.AudioComponent))
    (:else)))

(event EventEndPlay
  (params (EndPlayReason :type EEndPlayReason%20Enum :default "Destroyed"))
  ; End Play - tear down components and unbind from any in flight events.
  (|CallParentFunction EndPlayReason)
  (CallFunction|RemoveVFXAndUnbindEvents))

(event EventSpawnChestOpenVFX
  ; Spawn Chest Open VFX - fires on interact
  (CallFunction|RemoveVFXAndUnbindEvents))

(event OnLinkedActorDestroyed(FortLinkToActor)
  (params (DamageCauser :type Actor%20Object%20Reference))
  ; Spawn Projectile on Link Destroyed
  (CallFunction|HandleLinkedActorDestroyed))

(event EventOnLoot
  ; OnLoot
  (|CallParentFunction)
  (Utilities|FlowControl|Sequence
    (:then_0
      (Variables|Default|SetAllowFall false)
      (Collision|SetCollisionResponsetoChannel (Variables|Debug|GetStaticMeshComponent) "ECC_GameTraceChannel1" "ECR_Ignore") ; Fix issue where loot that bounces back into the container can't be grabbed
    )
    (:then_1
      ;[0.7135 0.3947 0.1212 0.5] FX when Opening chest
      (Niagara|SpawnSystemAttached "/CherrySmokeChestParent/Assets/FX/Proto/NS_CherrySmoke_Open_ChestBurst.NS_CherrySmoke_Open_ChestBurst" (Variables|Debug|GetStaticMeshComponent) "None" "0.000000,35.000000,80.000000" "0, 0, 0" "KeepRelativeOffset" true))))

(fn BlueprintCanInteract
  (params (InteractingPawn :type Fort%20Pawn%20Object%20Reference) (InteractionBeingAttempted :type EInteractionBeingAttempted%20Enum :default "FirstInteraction") (InteractionType :type TInteractionType%20Enum :default "IT_NoInteraction"))
  (return (|CallParentFunction InteractingPawn InteractionBeingAttempted InteractionType)))

(fn HandleLinkedActorDestroyed
  (bind (_returnvalue_x _returnvalue_y _returnvalue_z) (Math|Vector|BreakVector (Transformation|GetActorLocation)))
  (bind (_returnvalue_roll _returnvalue_pitch _returnvalue_yaw) (Math|Rotator|BreakRotator (Transformation|GetActorRotation)))
  ; Spawn Projectile on Link Destroyed
  (if (> (ScalableFloat|GetValueatLevel (Variables|Default|GetHF_EnableChestFall)) 0)
    (Variables|Actor|SetbCanBeDamaged true)
    (if (Variables|Default|GetAllowFall)
      (bind _returnvalue (Projectile|SpawnProjectile :ProjectileClass "/CherrySmokeChestParent/Gameplay/CheatSpawner/B_Prj_CherrySmoke_Chest.B_Prj_CherrySmoke_Chest_C" :RequestedBy self :SpawnLocation (Math|Vector|MakeVector _returnvalue_x _returnvalue_y (+ _returnvalue_z (Variables|Default|GetProjectileSpawnZOffset))) :SpawnRotation (Math|Rotator|MakeRotator 0.0 0.0 _returnvalue_yaw) :SpawnDirection (Math|Rotator|MakeRotfromY (Math|Vector|MakeVector _returnvalue_x _returnvalue_y (+ _returnvalue_z (Variables|Default|GetProjectileSpawnZOffset)))) :bUseDefaultPhysics false :InitialSpeed 7.5e+02 :GravityScale 0.7 :ChargePercent 0.0))
      (bind _asb_prj_cherry_smoke_chest (Utilities|Casting|CastToB_Prj_CherrySmoke_Chest _returnvalue))
      (Variables|/CherrySmokeChestParent/Gameplay/CheatSpawner/B_Prj_CherrySmoke_Chest.B_Prj_CherrySmoke_Chest_C|SetChest%20BGA (Utilities|GetClass self) _asb_prj_cherry_smoke_chest)
      (Default|CallOnSpawnProjectile _asb_prj_cherry_smoke_chest)
      (Actor|DestroyActor) ; Chest can sometimes become linked to wall behind it for some reason... This ensures it is destroyed regardless.
      (else
        (Actor|DestroyActor)))))

(fn SetCustomPingTimerForMarkerDisplayInfo
  (bind (_displayname _icon _customindicatorclass _customdetailsclass _primarycolor _sound _battachsound _screenclamping _specialcull _specialdotsizethreshold _specialdistance _customtimetolive _additionaldata) (Utilities|Struct|BreakMarkedActorDisplayInfo (Variables|Marking|GetMarkerDisplay)))
  (Variables|SetMarkerDisplay_DisplayName (Utilities|Struct|MakeMarkedActorDisplayInfo _displayname _icon _customindicatorclass _customdetailsclass _primarycolor _sound true _screenclamping _specialcull _specialdotsizethreshold _specialdistance (ScalableFloat|GetValueatLevel (Variables|Default|GetHF_PingTimer)) _additionaldata)))

(fn UpdateLockStatusVFX
  (params (RankNotMet :type Boolean))
  (bind _sound_indicator_component (Variables|Default|GetSound%20Indicator%20Component))
  (Niagara|SetNiagaraVariable(Bool) (Variables|Components|GetHeroRankUnlockedVFX) "Unlocked" (not RankNotMet))
  (bind _chest_ambient_sound (Variables|Default|GetChest_Ambient_Sound)
    (:then
      (if (or (Variables|GetRankNotMet) (Variables|Container|GetbAlreadySearched))
        (Audio|Components|Audio|FadeOut _chest_ambient_sound 0.5 0.0 "Linear" :!class /Script/Engine.AudioComponent)
        (SoundIndicator|StopSoundIndicatorLooping _sound_indicator_component)
        (else
          (Audio|Components|Audio|Play _chest_ambient_sound 0.0)
          (SoundIndicator|StartSoundIndicatorLooping _sound_indicator_component (Transformation|GetActorLocation) (Variables|/Script/Engine.AudioComponent|GetSound (Variables|Default|GetChest_Ambient_Sound)) 1.0 (Game|GetPlayerState 0)))))
    (:else)))

(fn ConstructionScript
  (|CallParentFunction))
