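; Base lock actor for CherrySmoke chests that gates opening behind an unlock check and pulls rank text and icon from its parent container.
; LockCheck drives prompt text, interact time, key visibility, and whether local interact triggers a search or failed-click ability with matching audio.
; Interrupting or opening fades looping interact audio.
; Opening also disables lock collision, container falling and one container collision channel so bounced-back loot stays grabbable.
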(event EventBeginPlay
  (|CallParentFunction)
  (CallFunction|SetInteractionDetails))

(event EventOpenContainers
  (|CallParentFunction)
  (Utilities|FlowControl|Sequence
    (:then_0
      (bind _interactaudio (Variables|Default|GetInteractAudio)
        (:then
          (Audio|Components|Audio|FadeOut _interactaudio 0.01 0.0 "Linear" :!class /Script/Engine.AudioComponent))
        (:else
          _))
      (SoundLibrary|PlayEvent (Variables|Default|GetSoundLibrary) "(TagName=\"SoundLibrary.Item.Interact.CherrySmoke.Open\")"))
    (:then_1
      (CallFunction|LockSetInteractCollision false) ; Fix issue where loot that bounces back into the container can't be grabbed
      (bind _asb_cherry_smoke_chest_container_base_parent (Utilities|Casting|CastToB_CherrySmokeChest_Container_BaseParent (Actor|GetParentActor)))
      (Variables|/CherrySmokeChestParent/Gameplay/B_CherrySmokeChest_Container_BaseParent.B_CherrySmokeChest_Container_BaseParent_C|SetAllowFall false _asb_cherry_smoke_chest_container_base_parent)
      (Collision|SetCollisionResponsetoChannel (Variables|/Script/FortniteGame.BuildingSMActor|GetStaticMeshComponent _asb_cherry_smoke_chest_container_base_parent) "ECC_GameTraceChannel1" "ECR_Ignore") ; Fix issue where loot that bounces back into the container can't be grabbed
    )))

(event Interaction|EventBlueprintOnInterruptInteract
  (|CallParentFunction)
  (bind _interacting_pawn (Variables|Default|GetInteracting%20Pawn)
    (:then
      (if (Pawn|IsLocallyViewed _interacting_pawn)
        (SoundLibrary|PlayEvent (Variables|Default|GetSoundLibrary) "(TagName=\"SoundLibrary.Item.Interact.Cancel\")")
        (bind _interactaudio (Variables|Default|GetInteractAudio)
          (:then
            (Audio|Components|Audio|FadeOut _interactaudio 0.01 0.0 "Linear" :!class /Script/Engine.AudioComponent))
          (:else))))
    (:else)))

(fn BlueprintDisplaySubInteractIcon
  (params (InteractingPawn :type Fort%20Pawn%20Object%20Reference) (InteractionBeingAttempted :type EInteractionBeingAttempted%20Enum :default "FirstInteraction") (InteractionType :type TInteractionType%20Enum :default "IT_NoInteraction"))
  (return true))

(fn BlueprintGetFailedSubTextTexture
  (params (InteractingPawn :type Fort%20Pawn%20Object%20Reference) (InteractionBeingAttempted :type EInteractionBeingAttempted%20Enum :default "FirstInteraction"))
  (|CallParentFunction InteractingPawn InteractionBeingAttempted)
  (return (Variables|Lock|GetSub%20Interaction%20Chest%20Rank%20Texture)))

(fn BlueprintGetInteractionString
  (params (InteractingPawn :type Fort%20Pawn%20Object%20Reference) (InteractionBeingAttempted :type EInteractionBeingAttempted%20Enum :default "FirstInteraction"))
  (bind _checkpassed (CallFunction|LockCheck InteractingPawn InteractionBeingAttempted))
  (if _checkpassed
    (return (Variables|Lock|GetInteraction_MeetsUnlockRequirements_Text))
    (else
      (return (Variables|Lock|GetInteraction_FailsUnlockRequirements_Text)))))

(fn BlueprintGetInteractionTime
  (params (InteractingPawn :type Fort%20Pawn%20Object%20Reference) (InteractionBeingAttempted :type EInteractionBeingAttempted%20Enum :default "FirstInteraction"))
  (bind (_outinteractiontime _returnvalue) (|CallParentFunction InteractingPawn InteractionBeingAttempted))
  (bind _checkpassed (CallFunction|LockCheck InteractingPawn InteractionBeingAttempted))
  (return (select _checkpassed 0.75 0.0) _returnvalue))

(fn BlueprintGetSubInteractionString
  (params (InteractingPawn :type Fort%20Pawn%20Object%20Reference) (InteractionBeingAttempted :type EInteractionBeingAttempted%20Enum :default "FirstInteraction"))
  (bind _checkpassed (CallFunction|LockCheck InteractingPawn InteractionBeingAttempted))
  (if (not _checkpassed)
    (return (Variables|Lock|GetSubInteraction_FailsRequirement_Text))))

(fn BlueprintOnLocalInteract
  (params (InteractingPawn :type Fort%20Player%20Pawn%20Object%20Reference) (InteractionBeingAttempted :type EInteractionBeingAttempted%20Enum :default "FirstInteraction"))
  (bind _returnvalue (Ability|GetFortAbilitySystemComponent InteractingPawn))
  (bind _returnvalue_2 (|CallParentFunction InteractingPawn InteractionBeingAttempted))
  (Utilities|FlowControl|Sequence
    (:then_0
      (Utilities|IsValid _returnvalue
        (:Is%20Valid
          (bind _checkpassed (CallFunction|LockCheck InteractingPawn InteractionBeingAttempted))
          (Abilities|TryActivateAbilitybyClass _returnvalue (select _checkpassed "/Game/Abilities/Player/Generic/Traits/DefaultPlayer/GA_DefaultPlayer_InteractSearch.GA_DefaultPlayer_InteractSearch_C" "/Game/Weapons/Prototype/ObjectMover/Blueprints/ObjectInteractionBehaviors/GA_Failed_Click.GA_Failed_Click_C") false)
          (if _checkpassed
            (bind _returnvalue_3 (SoundLibrary|PlayEvent (Variables|Default|GetSoundLibrary) "(TagName=\"SoundLibrary.Item.Interact.CherrySmoke.Long\")"))
            (Utilities|IsValid (Utilities|Array|Get(acopy) (.AudioComponents (Utilities|Struct|BreakSoundLibraryPlayResult _returnvalue_3)))
              (:Is%20Valid
                (bind _output_get (Variables|Default|SetInteractAudio (Utilities|Array|Get(acopy) (.AudioComponents (Utilities|Struct|BreakSoundLibraryPlayResult _returnvalue_3)))))
                (Audio|Components|Audio|FadeOut _output_get 0.1 0.0 "Linear" :!class /Script/Engine.AudioComponent)))
            (else
              (SoundLibrary|PlayEvent (Variables|Default|GetSoundLibrary) "(TagName=\"SoundLibrary.Item.Interact.CherrySmoke.Failed\")")
              (FailedAction|UnabletoPerformAction InteractingPawn true (LocalVariables|GetFailReason)))))))
    (:then_1
      (return _returnvalue_2))))

(fn BlueprintUpdateDynamicInteractionOptions
  (params (InteractingPawn :type Fort%20Pawn%20Object%20Reference) (InteractionBeingAttempted :type EInteractionBeingAttempted%20Enum :default "FirstInteraction") (InteractionType :type TInteractionType%20Enum :default "IT_NoInteraction"))
  (bind _outgameplaytags (|CallParentFunction InteractingPawn InteractionBeingAttempted InteractionType))
  (LocalVariables|SetInteractionOptions _outgameplaytags)
  (bind _checkpassed (CallFunction|LockCheck InteractingPawn InteractionBeingAttempted))
  (if (not _checkpassed)
    (GameplayTags|AddGameplayTag (LocalVariables|GetInteractionOptions) "(TagName=\"HUD.Athena.InteractionOptions.DisableInteractKey\")" :!class /Script/GameplayTags.BlueprintGameplayTagLibrary) ; Hide the Interact Key if interaction requirements fail
  )
  (return (LocalVariables|GetInteractionOptions)))

(fn SetInteractionDetails
  (bind _asb_cherry_smoke_chest_container_base_parent (Utilities|Casting|CastToB_CherrySmokeChest_Container_BaseParent (Actor|GetAttachParentActor)))
  (Variables|Lock|SetInteraction_FailsUnlockRequirements_Text (Variables|/CherrySmokeChestParent/Gameplay/B_CherrySmokeChest_Container_BaseParent.B_CherrySmokeChest_Container_BaseParent_C|GetInteraction_ChestRank_Text _asb_cherry_smoke_chest_container_base_parent))
  (Variables|Lock|SetSub%20Interaction%20Chest%20Rank%20Texture (Variables|/CherrySmokeChestParent/Gameplay/B_CherrySmokeChest_Container_BaseParent.B_CherrySmokeChest_Container_BaseParent_C|GetSubInteraction_ChestRank_Texture _asb_cherry_smoke_chest_container_base_parent)))

(fn ConstructionScript
  (|CallParentFunction))
